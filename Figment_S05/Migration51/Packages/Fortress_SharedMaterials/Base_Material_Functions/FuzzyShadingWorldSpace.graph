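# Material function that emulates velvet or moss-like fuzzy shading by modulating an input BaseColor with a view-angle term.
# It dots the camera vector with a world-space normal (pixel normal by default), multiplied by TwoSidedSign to handle backfaces, then clamps the result.
# CoreDarkness darkens camera-facing areas, while an inverted Fresnel-style term shaped by Power and scaled by EdgeBrightness brightens grazing edges.
# The two terms are summed and multiplied by BaseColor.
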MaterialFunction /Figment_S05/Migration51/Packages/Fortress_SharedMaterials/Base_Material_Functions/FuzzyShadingWorldSpace
graph {
 properties {description=This function emulates shading simmilar to velvet or moss, libraryCategoriesText [Misc,Shading]}
 nodes [
  n1 = FunctionInput(Preview=PixelNormalWS()) {bUsePreviewValueAsDefault=true, inputName=Normal, sortPriority=1, previewValue=<0 0 1 1>}
  n2 = TwoSidedSign()
  n3 = Clamp(None=DotProduct(A=CameraVectorWS(), B=Multiply(A=@n1, B=@n2)))
  n4 = FunctionInput() {bUsePreviewValueAsDefault=true, inputName=Power, inputType=FunctionInput_Scalar, sortPriority=3, previewValue=<6 0 0 1>}
  n5 = FunctionInput() {bUsePreviewValueAsDefault=true, inputName=EdgeBrightness, inputType=FunctionInput_Scalar, sortPriority=3, previewValue=<0.8 0 0 1>}
  n6 = Multiply(A=Power(Base=OneMinus(None=@n3), Exp=@n4), B=@n5)
  n7 = FunctionInput() {bUsePreviewValueAsDefault=true, inputName=CoreDarkness, inputType=FunctionInput_Scalar, sortPriority=2, previewValue=<0.8 0 0 1>}
  n8 = Add(A=OneMinus(None=Multiply(A=@n3, B=@n7)), B=Constant())
  FunctionOutput(None=Multiply(A=Add(A=@n6, B=@n8), B=FunctionInput() {inputName=BaseColor, previewValue=<0.5 0.5 0.5 1>}))
 ]
}
