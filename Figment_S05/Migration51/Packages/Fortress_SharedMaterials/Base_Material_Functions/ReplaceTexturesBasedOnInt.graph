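# Material function that selects one of up to six inputs, meant for textures, using the scalar transitionFactor.
# Chained If nodes step from one input to the next at whole-number thresholds, with hard cuts rather than blends.
# "Use N Textures" static switches set the chain depth, while Default Color and MaskFirstMaterial affect the first slot.
# A second chain, matching half-unit-centered bands with Default Color fallback, is not connected to Final Output.
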MaterialFunction /Figment_S05/Migration51/Packages/Fortress_SharedMaterials/Base_Material_Functions/ReplaceTexturesBasedOnInt
graph {nodes [
  n1 = Constant() {r=5}
  n2 = FunctionInput() {inputName=transitionFactor, inputType=FunctionInput_Scalar, sortPriority=-1}
  n3 = FunctionInput() {inputName=1}
  n4 = FunctionInput() {inputName=Default Color}
  n5 = If(A=Constant() {r=1}, "A > B"=@n3, "A == B"=@n3, "A < B"=@n4) {desc=0, equalsThreshold=0.5}
  n6 = FunctionInput(Preview=StaticBool() {value=true}) {bUsePreviewValueAsDefault=true, inputName=MaskFirstMaterial, inputType=FunctionInput_StaticBool, sortPriority=10}
  n7 = StaticBool()
  n8 = FunctionInput(Preview=@n7) {bUsePreviewValueAsDefault=true, inputName=Use 1 Textures, inputType=FunctionInput_StaticBool, sortPriority=10}
  n9 = FunctionInput() {inputName=2, sortPriority=1}
  n10 = If(
   A=Constant() {r=1}
   B=@n2
   "A > B"=StaticSwitch(True=StaticSwitch(True=@n5, False=@n3, Value=@n6) {desc=Isolate texture 1}, False=@n4, Value=@n8) {desc=use texture 1}
   "A < B"=@n9
  ) {desc=0}
  n11 = FunctionInput() {inputName=3, sortPriority=2}
  n12 = If(A=Constant() {r=2}, B=@n2, "A > B"=@n10, "A < B"=@n11)
  n13 = FunctionInput() {inputName=4, sortPriority=3}
  n14 = If(A=Constant() {r=3}, B=@n2, "A > B"=@n12, "A < B"=@n13)
  n15 = FunctionInput() {inputName=5, sortPriority=4}
  n16 = If(A=Constant() {r=4}, B=@n2, "A > B"=@n14, "A < B"=@n15)
  n17 = FunctionInput() {inputName=6, sortPriority=5}
  n18 = FunctionInput(Preview=@n7) {bUsePreviewValueAsDefault=true, inputName=Use 2 Textures, inputType=FunctionInput_StaticBool, sortPriority=10}
  n19 = FunctionInput(Preview=@n7) {bUsePreviewValueAsDefault=true, inputName=Use 3 Textures, inputType=FunctionInput_StaticBool, sortPriority=11}
  n20 = FunctionInput(Preview=@n7) {bUsePreviewValueAsDefault=true, inputName=Use 4 Textures, inputType=FunctionInput_StaticBool, sortPriority=12}
  n21 = StaticSwitch(True=@n14, False=StaticSwitch(True=@n12, False=StaticSwitch(True=@n10, False=@n3, Value=@n18), Value=@n19), Value=@n20)
  n22 = FunctionInput(Preview=@n7) {bUsePreviewValueAsDefault=true, inputName=Use 5 Textures, inputType=FunctionInput_StaticBool, sortPriority=13}
  n23 = FunctionInput(Preview=@n7) {bUsePreviewValueAsDefault=true, inputName=Use 6 Textures, inputType=FunctionInput_StaticBool, sortPriority=14}
  FunctionOutput(None=StaticSwitch(True=If(A=@n1, B=@n2, "A > B"=@n16, "A < B"=@n17), False=StaticSwitch(True=@n16, False=@n21, Value=@n22), Value=@n23)) {outputName=Final Output, sortPriority=-2}
  n24 = Constant() {r=0.5}
  n25 = StaticSwitch(
   True=StaticSwitch(True=If(A=@n24, B=@n2, "A > B"=@n4, "A == B"=@n3, "A < B"=@n4) {desc=0, equalsThreshold=0.5}, False=@n3, Value=@n6) {desc=Isolate texture 1}
   False=@n4
   Value=@n8
  ) {desc=use texture 1}
  n26 = StaticSwitch(True=If(A=Constant() {r=1.5}, B=@n2, "A > B"=@n25, "A == B"=@n9, "A < B"=@n25) {equalsThreshold=0.5}, False=@n25, Value=@n18)
  n27 = StaticSwitch(True=If(A=Constant() {r=2.5}, B=@n2, "A > B"=@n26, "A == B"=@n11, "A < B"=@n26) {equalsThreshold=0.5}, False=@n26, Value=@n19)
  n28 = StaticSwitch(True=If(A=Constant() {r=3.5}, B=@n2, "A > B"=@n27, "A == B"=@n13, "A < B"=@n27) {equalsThreshold=0.5}, False=@n27, Value=@n20)
  n29 = StaticSwitch(True=If(A=Constant() {r=4.5}, B=@n2, "A > B"=@n28, "A == B"=@n15, "A < B"=@n28) {equalsThreshold=0.5}, False=@n28, Value=@n22)
  StaticSwitch(True=If(A=Constant() {r=5.5}, B=@n2, "A > B"=@n29, "A == B"=@n17, "A < B"=@n29) {equalsThreshold=0.5}, False=@n29, Value=@n23)
 ]}
